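# Niagara aura effect for Sideways chests, layering purple lightning, vapor and light.
# Lightning spawns short-lived mesh flickers pushed outward, with curve-driven HDR color and randomized dynamic material parameters.
# Vapor_fast_thin emits spinning sprite puffs that fly outward under drag, scale by curve and fade.
# Vapor_wide001 bursts one persistent, lifetime-looping particle that feeds two mesh renderers, a component renderer and a raised light, acting as the steady glow.
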NiagaraSystem /SidewaysGameplay/Environmental/SidewaysChests/Assets/FX/NS_SidewaysChest_Aura
dependencies {
 usedDataInterfaces [&NiagaraDataInterfaceCurve, &NiagaraDataInterfaceVectorCurve]
 usedDynamicInputs []
 usedModules [&SystemState]
}
emitters {
 Lightning {
  emitterUpdateScript [
   &EmitterState(Life Cycle Mode=::NewEnumerator1, UseLoopDelay=false)
   &SpawnBurst_Instantaneous(Spawn Count=4, Use Spawn Probability=false, UseLoopCountLimit=false) {enabled=false}
   &SpawnRate(Spawn Probability=0.3, Use Spawn Probability=true)
  ]
  particleSpawnScript [
   &InitializeParticle(
    Color=<rgba 0.0701 0 1 0.5>
    Lifetime Max=0.2
    Lifetime Min=0.03
    Lifetime Mode=::NewEnumerator1
    Mesh Scale Mode=::NewEnumerator1
    Mesh Uniform Scale Max=0.3
    Mesh Uniform Scale Min=0.2
    UsePositionOffset=false
    Write Material Random=true
   )
   &SphereLocation(
    MaskSpawnGroup=false
    Offset=<0 0 25>
    Override Seed=false
    Sphere Radius=5
    Surface Only=false
    UseNonUniformScale=false
    UseOffset=true
    UseSphereOrientation=false
   )
   &AddVelocityFromPoint(
    Module.Use Falloff Distance=false
    Origin Offset=<0 0 -5>
    Velocity Strength=&RandomRangeFloat(Minimum=55, Maximum=125, Randomness Mode=::NewEnumerator2, Random Seed=?)
   )
  ]
  particleUpdateScript [
   &ParticleState()
   &Drag(Use Linear Drag=true, Use Rotational Drag=false)
   &SolveForcesAndVelocity(Clamp Velocity=false, Limit Acceleration=false)
   &ScaleColor(
    Scale Alpha=&FloatFromCurve(FloatCurve=DataInterface() {propertyValues {Curve {keys [<at 0.2 1>,<at 0.4 0.1>,<at 1 0>]}}})
    Scale RGB=&VectorFromCurve(
     VectorCurve=DataInterface() {propertyValues {
       XCurve {keys [<at 0 12>,<at 0.1041 1.3151>,<at 0.3 0>]}
       YCurve {keys [<at 0 0.5>,<at 0.1041 0.8651>,<at 0.3 0>]}
       ZCurve {keys [<at 0 12>,<at 0.1223 14.0295>,<at 0.3 0>]}
      }}
     Scale Curve=<5 5 5>
    )
    ScaleA=true
    ScaleRGB=true
   )
   &DynamicMaterialParameters(
    Index 0 Param 1=0.1
    Index 0 Param 2=&RandomRangeFloat(Minimum=0, Maximum=1)
    Index 0 Param 3=&RandomRangeFloat(Minimum=0.8, Maximum=2, Evaluation Type=::NewEnumerator1)
    Module.Param0WriteEnabled=true
    Module.Param1WriteEnabled=true
    Module.Param2WriteEnabled=true
    Module.Param3WriteEnabled=true
   )
   &SpriteRotationRate(
    Delta Time:NiagaraFloat=&FloatFromCurve(
     FloatCurve=DataInterface() {propertyValues {Curve {keys [<at 0 1>,<at 0.3 0.25>,<at 1 0>]}}}
     Scale Curve=&RandomRangeFloat(Minimum=-1, Maximum=1)
    )
    Rotation Rate=2
   )
  ]
  rendererClasses [&NiagaraMeshRendererProperties]
  renderers [{rendererClass=&NiagaraMeshRendererProperties, rendererIndex=0}]
 }
 Vapor_fast_thin {
  emitterUpdateScript [
   &EmitterState(Life Cycle Mode=::NewEnumerator1, UseLoopDelay=false)
   &SpawnBurst_Instantaneous(
    Spawn Count=1
    Spawn Probability=&MakeCustomFloatFromBool(
     Bool=&SetBoolByIntComparison(A=$System.LoopCount, B=1, Comparison Type=::NewEnumerator6)
     True Float=1
     False Float=0
    )
    Use Spawn Probability=true
   ) {enabled=false, module=SpawnBurst_Instantaneous001}
   &SpawnBurst_Instantaneous(Spawn Count=4, Use Spawn Probability=false, UseLoopCountLimit=false) {enabled=false}
   &SpawnRate(SpawnRate=1.5, Use Spawn Probability=false)
  ]
  particleSpawnScript [
   &InitializeParticle(
    Color=<rgba 0.2217 0.1178 1 2>
    Lifetime Max=1.5
    Lifetime Mode=::NewEnumerator1
    Sprite Size Mode=::NewEnumerator1
    Uniform Sprite Size Max=55
    Uniform Sprite Size Min=35
    UsePositionOffset=false
    Write Material Random=true
   )
   &SphereLocation(
    MaskSpawnGroup=false
    Offset=<0 0 16>
    Override Seed=false
    Sphere Radius=5
    Surface Only=false
    UseNonUniformScale=false
    UseOffset=true
    UseSphereOrientation=false
   )
   &AddVelocityFromPoint(
    Module.Use Falloff Distance=false
    Origin Offset=<0 0 -5>
    Velocity Strength=&RandomRangeFloat(Minimum=80, Maximum=150, Randomness Mode=::NewEnumerator2, Random Seed=?)
   )
   /NiagaraFN/Sprite/FN_RandomizeSpriteUV() {enabled=false}
  ]
  particleUpdateScript [
   &ParticleState()
   &Drag(Use Linear Drag=true, Use Rotational Drag=false)
   &SolveForcesAndVelocity(Clamp Velocity=false, Limit Acceleration=false)
   &SpriteFacingAndAlignment(WriteAlignment=false, WriteFacing=true) {enabled=false}
   &FN_Sprite_Attributes(Create Sprite Rotation Attribute=true, SpriteRotation=1) {enabled=false}
   &ScaleSpriteSize(
    Uniform Curve Scale=5
    Uniform Curve Sprite Scale=DataInterface() {propertyValues {Curve {keys [<at 0 0.5 RCIM_Cubic>, {arriveTangent=0.3, interpMode=RCIM_Cubic, leaveTangent=0.3, time=0.6, value=0.6}, <at 1 0.8 RCIM_Cubic>]}}}
   )
   &ScaleColor(
    Scale Alpha=&FloatFromCurve(FloatCurve=DataInterface() {propertyValues {Curve {keys [<at 0.2 1>,<at 0.5 0.15>,<at 1 0>]}}}, Scale Curve=0.1)
    Scale RGB=<6 1 7>
    ScaleA=true
    ScaleRGB=true
   )
   &SpriteRotationRate(
    Delta Time:NiagaraFloat=&FloatFromCurve(
     FloatCurve=DataInterface() {propertyValues {Curve {keys [<at 0 1>,<at 0.3 0.25>,<at 1 0>]}}}
     Scale Curve=&RandomRangeFloat(Minimum=-1, Maximum=1)
    )
    Rotation Rate=3
   )
  ]
  rendererClasses [&NiagaraSpriteRendererProperties]
  renderers [{rendererClass=&NiagaraSpriteRendererProperties}]
 }
 Vapor_wide001 {
  emitterUpdateScript [
   &EmitterState(Inactive Response=::NewEnumerator2, Life Cycle Mode=::NewEnumerator1, Loop Behavior=::NewEnumerator1, Loop Duration=126, UseLoopDelay=false)
   &SpawnBurst_Instantaneous(
    Spawn Count=1
    Spawn Probability=&MakeCustomFloatFromBool(
     Bool=&SetBoolByIntComparison(A=$System.LoopCount, B=1, Comparison Type=::NewEnumerator6)
     True Float=1
     False Float=0
    )
    Use Spawn Probability=true
   ) {enabled=false}
   &SpawnBurst_Instantaneous(Spawn Count=1, Use Spawn Probability=false) {module=SpawnBurst_Instantaneous001}
  ]
  particleSpawnScript [
   &InitializeParticle(
    Color=<rgba 3 0.3097 3 0.3>
    Mesh Scale Mode=::NewEnumerator3
    Mesh Uniform Scale=&RandomRangeFloat(Minimum=1, Maximum=1)
    UsePositionOffset=false
    Write Material Random=true
   )
   &SphereLocation(
    MaskSpawnGroup=false
    Override Seed=false
    Sphere Radius=1
    Surface Only=false
    UseNonUniformScale=false
    UseOffset=true
    UseSphereOrientation=false
   )
   /SetVariables(Particles.LightPosition:Vector3f=<0 0 120>, Particles.LightRadius=125, Particles.LightScale:Vector3f=<1 1 1>) {isSetParameters=true}
  ]
  particleUpdateScript [
   &ParticleState(Kill Particles When Lifetime Has Elapsed=false, Loop Particles Lifetime=true)
   /SetVariables(Particles.SpriteSize=<333 333>) {enabled=false, isSetParameters=true}
   &CameraOffset(Camera Offset Amount=-30)
   /SetVariables(Particles.ColorB:LinearColor=&MultiplyLinearColor(Linear Color=$Particles.Color, Scale Factor=<2 2 2 3>)) {isSetParameters=true}
  ]
  rendererClasses [
   &NiagaraMeshRendererProperties
   &NiagaraComponentRendererProperties
   &NiagaraLightRendererProperties
  ]
  renderers [
   {rendererClass=&NiagaraMeshRendererProperties, rendererIndex=0}
   {rendererClass=&NiagaraMeshRendererProperties, rendererIndex=1}
   {rendererClass=&NiagaraComponentRendererProperties, rendererIndex=2}
   {rendererClass=&NiagaraLightRendererProperties, rendererIndex=3}
  ]
 }
}
summary {
 systemName=NS_SidewaysChest_Aura
 emitters [
  {emitterName=Vapor_fast_thin, rendererClasses [&NiagaraSpriteRendererProperties]}
  {emitterName=Lightning, rendererClasses [&NiagaraMeshRendererProperties]}
  {
   emitterName=Vapor_wide001
   rendererClasses [
    &NiagaraMeshRendererProperties
    &NiagaraComponentRendererProperties
    &NiagaraLightRendererProperties
   ]
  }
 ]
 userVariables []
}
