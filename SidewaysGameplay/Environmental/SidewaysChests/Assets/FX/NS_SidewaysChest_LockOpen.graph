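# Niagara effect for a Sideways chest lock opening, built from burst emitters.
# An orange light flash fades while mesh rays grow and fade, driven by dynamic material parameters, alongside a separate offset mesh.
# GPU rune sprites burst outward along negative Y, swirling through panning curl noise and spinning, both scaled by speed.
# Halo, vertical beam and duplicate ray and rune emitters are disabled; no user parameters are exposed.
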NiagaraSystem /SidewaysGameplay/Environmental/SidewaysChests/Assets/FX/NS_SidewaysChest_LockOpen
dependencies {
 usedDataInterfaces [&NiagaraDataInterfaceCurve, &NiagaraDataInterfaceColorCurve]
 usedDynamicInputs []
 usedModules [&SystemState]
}
emitters {
 Halo_and_Glow {
  bEnabled=false
  emitterUpdateScript [
   &EmitterState()
   &SpawnRate(Use Spawn Probability=false) {enabled=false}
   &SpawnBurst_Instantaneous(Spawn Count=1, Use Spawn Probability=false)
  ]
  particleSpawnScript [&InitializeParticle(Color Mode=::NewEnumerator0, Lifetime=0.7, UsePositionOffset=false, Write Material Random=true)]
  particleUpdateScript [
   &ParticleState()
   &Color(Color=&ColorFromCurve(ColorCurve=DataInterface() {propertyValues {
       AlphaCurve {keys [<at 0.1047 0.3 RCIM_Cubic>,<at 0.9745 0 RCIM_Cubic>]}
       BlueCurve {keys [<at 0.2077 0.1324>,<at 0.5016 0.6667 RCIM_Cubic>]}
       GreenCurve {keys [<at 0.2077 0.0208>,<at 0.5016 0.2049 RCIM_Cubic>]}
       RedCurve {keys [<at 0.2077 0.6823>,<at 0.5016 0.6624 RCIM_Cubic>]}
      }}))
   /SetVariables(Particles.SpriteSize=&Vector2DFromFloat(Value=&FloatFromCurve(
      FloatCurve=DataInterface() {propertyValues {Curve {keys [
          {arriveTangent=0.3463, interpMode=RCIM_Cubic, leaveTangent=0.3463, tangentMode=RCTM_User, time=0, value=0.2}
          {arriveTangent=1.8357, interpMode=RCIM_Cubic, leaveTangent=1.8357, tangentMode=RCTM_User, time=0.3, value=0.95}
          {arriveTangent=-0.0405, interpMode=RCIM_Cubic, leaveTangent=-0.0405, tangentMode=RCTM_User, time=1, value=1.5}
         ]}}}
      Scale Curve=160
     ))) {isSetParameters=true, module=SetVariables_#1}
   &CameraOffset(Camera Offset Amount=120)
   &FN_Radial_Material(
    Radial Mask Power=1
    Radial Mask Tile U=0.2
    Radial Mask Tile V=0.5
    Sprite Extrusion Speed:NiagaraFloat=&FloatFromCurve(
     FloatCurve=DataInterface() {propertyValues {Curve {keys [
         {arriveTangent=-0.2256, interpMode=RCIM_Cubic, time=0, value=1.75}
         {arriveTangent=2.75, interpMode=RCIM_Cubic, leaveTangent=2.75, time=0.3, value=2.375}
         {arriveTangent=2.5, interpMode=RCIM_Cubic, leaveTangent=2.5, time=0.5, value=3.125}
         {interpMode=RCIM_Cubic, leaveTangent=-0.5666, time=0.7, value=3.375}
        ]}}}
     Scale Curve=-0.3
    )
    Sprite Spin Speed:NiagaraFloat=&FloatFromCurve(FloatCurve=DataInterface() {propertyValues {Curve {keys [<at 0 0.2>,<at 0.4 1.125>,<at 1 2.25>]}}}, Scale Curve=0.1)
   ) {enabled=false}
  ]
  rendererClasses [&NiagaraSpriteRendererProperties]
  renderers [
   {rendererClass=&NiagaraSpriteRendererProperties}
   {rendererClass=&NiagaraSpriteRendererProperties, rendererIndex=1}
   {rendererClass=&NiagaraSpriteRendererProperties, rendererIndex=2}
  ]
 }
 LightRenderer {
  emitterUpdateScript [
   &EmitterState(Life Cycle Mode=::NewEnumerator1, Loop Behavior=::NewEnumerator1, Loop Duration=2, UseLoopDelay=false)
   &SpawnBurst_Instantaneous(Use Spawn Probability=false, UseLoopCountLimit=false)
  ]
  particleSpawnScript [
   &InitializeParticle(
    Color=<rgba 1 0.1157 0 15>
    Lifetime=0.5
    Sprite Size Mode=::NewEnumerator1
    Uniform Sprite Size Max=100
    Uniform Sprite Size Min=55
    UsePositionOffset=false
    Write Material Random=true
   )
   &SphereLocation(
    MaskSpawnGroup=false
    Override Seed=false
    Sphere Radius=2
    Surface Only=false
    UseNonUniformScale=false
    UseOffset=false
    UseSphereOrientation=false
   )
   /SetVariables(Particles.LightRadius=80, Particles.LightScale:Vector3f=<1 1 1>) {isSetParameters=true}
  ]
  particleUpdateScript [
   &ParticleState()
   &ScaleColor(
    Scale Alpha=&FloatFromCurve(FloatCurve=DataInterface() {propertyValues {Curve {keys [<at 0.2 2>,<at 0.7 0.5>,<at 1 0>]}}})
    ScaleA=true
    ScaleRGB=true
   )
  ]
  rendererClasses [&NiagaraLightRendererProperties]
  renderers [{rendererClass=&NiagaraLightRendererProperties, rendererIndex=0}]
 }
 Mesh_rays {
  bEnabled=false
  emitterUpdateScript [
   &EmitterState(Life Cycle Mode=::NewEnumerator1, Loop Behavior=::NewEnumerator1, Loop Duration=2, UseLoopDelay=false)
   &SpawnBurst_Instantaneous(Spawn Count=5, Use Spawn Probability=false, UseLoopCountLimit=false)
  ]
  particleSpawnScript [
   &InitializeParticle(
    Color=<rgba 1 0.1087 0 1>
    Lifetime Max=1
    Lifetime Min=0.5
    Lifetime Mode=::NewEnumerator1
    Lifetime Randomness Mode:ENiagaraRandomnessMode=::NewEnumerator2
    Mesh Scale Mode=::NewEnumerator3
    Mesh Uniform Scale=8
    Position Offset=<0 -8 -15>
    Sprite Size Mode=::NewEnumerator1
    Uniform Sprite Size Max=1
    Uniform Sprite Size Min=1
    UsePositionOffset=true
    Write Material Random=true
   )
   /Niagara/Modules/Update/Orientation/V2/OrientMeshToVector(Look At Direction:Vector3f=<0 -1 0>, Override Quat=false, UseOptionalSideAxis=false, UseOptionalUpAxis=false)
   &SphereLocation(
    MaskSpawnGroup=false
    Offset=<0 0 -5>
    Sphere Coordinate Space=::Simulation
    Sphere Distribution:ENiagaraSphereDistributionMode=::NewEnumerator3
    Sphere Radius=5
    UseNonUniformScale=false
    UseOffset=false
    UseSphereOrientation=false
   ) {enabled=false, module=SphereLocation001}
   &SphereLocation(
    MaskSpawnGroup=false
    Offset=<0 20 0>
    Sphere Distribution:ENiagaraSphereDistributionMode=::NewEnumerator1
    Sphere Radius=0.0001
    UseNonUniformScale=false
    UseOffset=true
    UseSphereOrientation=false
   )
   &self:ScratchModule_01(Magnitude=5, Range:NiagaraFloat=&RandomRangeFloat(Minimum=-40, Maximum=40), Yaw=-90) {enabled=false, module=ScratchModule_01}
  ]
  particleUpdateScript [
   &ParticleState()
   &ScaleColor(
    Scale Alpha=&FloatFromCurve(FloatCurve=DataInterface() {propertyValues {Curve {keys [<at 0.2 1>,<at 0.4 0.35>,<at 0.6 0.05>,<at 1 0>]}}})
    ScaleA=true
    ScaleRGB=false
   )
   &DynamicMaterialParameters(
    Index 0 Param 1=&FloatFromCurve(FloatCurve=DataInterface() {propertyValues {Curve {keys [
         {arriveTangent=-0.0674, interpMode=RCIM_Cubic, leaveTangent=-0.0674, tangentMode=RCTM_User, time=0.4, value=0}
         {arriveTangent=4.1799, interpMode=RCIM_Cubic, leaveTangent=4.1799, tangentMode=RCTM_User, time=0.8, value=1.125}
         {arriveTangent=-0.0405, interpMode=RCIM_Cubic, leaveTangent=-0.0405, tangentMode=RCTM_User, time=1, value=0}
        ]}}})
    Index 0 Param 2=0.3
    Index 0 Param 3=0.2
    Module.Param0WriteEnabled=true
    Module.Param1WriteEnabled=true
    Module.Param2WriteEnabled=true
    Module.Param3WriteEnabled=false
   )
   &self:ScratchModule_01(
    Magnitude:NiagaraFloat=$Particles.NormalizedAge
    Pitch=20
    Range:NiagaraFloat=&RandomRangeFloat(Minimum=-60, Maximum=60)
    Yaw=-90
   ) {module=ScratchModule_01001}
   &ScaleMeshSize(Scale Factor=&VectorFromFloat(Value=&RandomRangeFloat(Minimum=0.5, Maximum=1))) {enabled=false}
   &ScaleMeshSize(Scale Factor=&VectorFromFloat(Value=&FloatFromCurve(
      FloatCurve=DataInterface() {propertyValues {Curve {keys [{interpMode=RCIM_Cubic, tangentMode=RCTM_User, time=0, value=0}, {interpMode=RCIM_Cubic, tangentMode=RCTM_User, time=1, value=1}]}}}
      Scale Curve=&RandomRangeFloat(Minimum=0.5, Maximum=1)
     ))) {module=ScaleMeshSize001}
  ]
  rendererClasses [&NiagaraSpriteRendererProperties, &NiagaraMeshRendererProperties]
  renderers [
   {rendererClass=&NiagaraSpriteRendererProperties}
   {rendererClass=&NiagaraMeshRendererProperties, rendererIndex=1}
   {rendererClass=&NiagaraMeshRendererProperties, rendererIndex=2}
  ]
 }
 Mesh_rays001 {
  emitterUpdateScript [
   &EmitterState(Life Cycle Mode=::NewEnumerator1, Loop Behavior=::NewEnumerator1, Loop Duration=2, UseLoopDelay=false)
   &SpawnBurst_Instantaneous(Spawn Count=5, Use Spawn Probability=false, UseLoopCountLimit=false)
  ]
  particleSpawnScript [
   &InitializeParticle(
    Color=<rgba 1 0.1087 0 1>
    Lifetime Max=1
    Lifetime Min=0.5
    Lifetime Mode=::NewEnumerator1
    Lifetime Randomness Mode:ENiagaraRandomnessMode=::NewEnumerator2
    Mesh Scale Mode=::NewEnumerator3
    Mesh Uniform Scale=8
    Position Offset=<0 -8 -15>
    Sprite Size Mode=::NewEnumerator1
    Uniform Sprite Size Max=1
    Uniform Sprite Size Min=1
    UsePositionOffset=true
    Write Material Random=true
   )
   &OrientMeshToVector(Look At Direction:Vector3f=<0 -1 0>, Override Quat=false, UseOptionalSideAxis=false, UseOptionalUpAxis=false)
   &SphereLocation(
    MaskSpawnGroup=false
    Offset=<0 0 -5>
    Sphere Coordinate Space=::Simulation
    Sphere Distribution:ENiagaraSphereDistributionMode=::NewEnumerator3
    Sphere Radius=5
    UseNonUniformScale=false
    UseOffset=false
    UseSphereOrientation=false
   ) {enabled=false, module=SphereLocation001}
   &SphereLocation(
    MaskSpawnGroup=false
    Offset=<0 20 -10>
    Sphere Distribution:ENiagaraSphereDistributionMode=::NewEnumerator1
    Sphere Radius=0.0001
    UseNonUniformScale=false
    UseOffset=true
    UseSphereOrientation=false
   )
   &self:ScratchModule_01(Magnitude=5, Range:NiagaraFloat=&RandomRangeFloat(Minimum=-40, Maximum=40), Yaw=-90) {enabled=false, module=ScratchModule_01}
  ]
  particleUpdateScript [
   &ParticleState()
   &ScaleColor(
    Scale Alpha=&FloatFromCurve(FloatCurve=DataInterface() {propertyValues {Curve {keys [<at 0.2 1>,<at 0.4 0.35>,<at 0.6 0.05>,<at 1 0>]}}})
    ScaleA=true
    ScaleRGB=false
   )
   &DynamicMaterialParameters(
    Index 0 Param 1=&FloatFromCurve(FloatCurve=DataInterface() {propertyValues {Curve {keys [
         {arriveTangent=-0.0674, interpMode=RCIM_Cubic, leaveTangent=-0.0674, tangentMode=RCTM_User, time=0.4, value=0}
         {arriveTangent=4.1799, interpMode=RCIM_Cubic, leaveTangent=4.1799, tangentMode=RCTM_User, time=0.8, value=1.125}
         {arriveTangent=-0.0405, interpMode=RCIM_Cubic, leaveTangent=-0.0405, tangentMode=RCTM_User, time=1, value=0}
        ]}}})
    Index 0 Param 2=0.3
    Index 0 Param 3=0.2
    Module.Param0WriteEnabled=true
    Module.Param1WriteEnabled=true
    Module.Param2WriteEnabled=true
    Module.Param3WriteEnabled=false
   )
   &self:ScratchModule_01(
    Magnitude:NiagaraFloat=$Particles.NormalizedAge
    Pitch=20
    Range:NiagaraFloat=&RandomRangeFloat(Minimum=-60, Maximum=60)
    Yaw=-90
   ) {module=ScratchModule_01001}
   &ScaleMeshSize(Scale Factor=&VectorFromFloat(Value=&RandomRangeFloat(Minimum=0.5, Maximum=1))) {enabled=false}
   &ScaleMeshSize(Scale Factor=&VectorFromFloat(Value=&FloatFromCurve(
      FloatCurve=DataInterface() {propertyValues {Curve {keys [{interpMode=RCIM_Cubic, tangentMode=RCTM_User, time=0, value=0}, {interpMode=RCIM_Cubic, tangentMode=RCTM_User, time=1, value=1}]}}}
      Scale Curve=&RandomRangeFloat(Minimum=0.5, Maximum=1)
     ))) {module=ScaleMeshSize001}
  ]
  rendererClasses [&NiagaraSpriteRendererProperties, &NiagaraMeshRendererProperties]
  renderers [
   {rendererClass=&NiagaraSpriteRendererProperties}
   {rendererClass=&NiagaraMeshRendererProperties, rendererIndex=1}
   {rendererClass=&NiagaraMeshRendererProperties, rendererIndex=2}
  ]
 }
 Mesh_rays_offset {
  emitterUpdateScript [
   &EmitterState(Life Cycle Mode=::NewEnumerator1, Loop Behavior=::NewEnumerator1, Loop Duration=2, UseLoopDelay=false)
   &SpawnBurst_Instantaneous(Use Spawn Probability=false, UseLoopCountLimit=false)
  ]
  particleSpawnScript [
   &InitializeParticle(Mesh Scale Mode=::NewEnumerator3, UsePositionOffset=false, Write Material Random=true)
   &FN_Mesh_Attributes() {enabled=false, module=FN_Mesh_Attributes001}
   &self:AdjustTransform(PositionOffset:Vector3f=<0 15 5>, Roll=2.6429, Yaw=-180) {module=AdjustTransform001}
   &self:AdjustTransform(PositionOffset:Vector3f=<0 5.1639 1.3795>, Roll=2.6429, Yaw=-180) {enabled=false, module=AdjustTransform002}
   &self:AdjustTransform(PositionOffset:Vector3f=<0 1 0>) {enabled=false, module=AdjustTransform}
  ]
  particleUpdateScript [
   &ParticleState()
   &ScaleColor(
    Scale Alpha=&FloatFromCurve(FloatCurve=DataInterface() {propertyValues {Curve {keys [
         {arriveTangent=9.0725, interpMode=RCIM_Cubic, leaveTangent=9.0725, tangentMode=RCTM_User, time=0, value=1.05}
         {arriveTangent=-2.5249, interpMode=RCIM_Cubic, leaveTangent=-2.5249, tangentMode=RCTM_User, time=0.3, value=1.975}
         {arriveTangent=-0.0405, interpMode=RCIM_Cubic, leaveTangent=-0.0405, tangentMode=RCTM_User, time=1, value=0}
        ]}}})
    ScaleA=true
    ScaleRGB=false
   )
   &DynamicMaterialParameters(
    Index 0 Param 1=&FloatFromCurve(FloatCurve=DataInterface() {propertyValues {Curve {keys [
         {arriveTangent=-0.0674, interpMode=RCIM_Cubic, leaveTangent=-0.0674, tangentMode=RCTM_User, time=0.4, value=0}
         {arriveTangent=4.1799, interpMode=RCIM_Cubic, leaveTangent=4.1799, tangentMode=RCTM_User, time=0.8, value=1.125}
         {arriveTangent=-0.0405, interpMode=RCIM_Cubic, leaveTangent=-0.0405, tangentMode=RCTM_User, time=1, value=0}
        ]}}})
    Index 0 Param 2=0.3
    Index 0 Param 3=0.2
    Module.Param0WriteEnabled=true
    Module.Param1WriteEnabled=true
    Module.Param2WriteEnabled=true
    Module.Param3WriteEnabled=false
   )
   &FN_Mesh_Attributes(Create Orientation Attribute=true, Create Scale Attribute=true, Orientation:Quat4f=$Engine.Owner.Rotation, Scale=<1.4 1.4 1.4>) {enabled=false}
  ]
  rendererClasses [&NiagaraMeshRendererProperties]
  renderers [{rendererClass=&NiagaraMeshRendererProperties, rendererIndex=0}]
 }
 Mesh_rays_offset001 {
  bEnabled=false
  emitterUpdateScript [
   &EmitterState(Life Cycle Mode=::NewEnumerator1, Loop Behavior=::NewEnumerator1, Loop Duration=2, UseLoopDelay=false)
   &SpawnBurst_Instantaneous(Use Spawn Probability=false, UseLoopCountLimit=false)
  ]
  particleSpawnScript [
   &InitializeParticle(Mesh Scale Mode=::NewEnumerator3, UsePositionOffset=false, Write Material Random=true)
   &FN_Mesh_Attributes() {enabled=false, module=FN_Mesh_Attributes001}
   &self:AdjustTransform(PositionOffset:Vector3f=<0 10 0>, Roll=2.6429, Yaw=-180) {enabled=false, module=AdjustTransform001}
   &self:AdjustTransform(PositionOffset:Vector3f=<0 5.1639 1.3795>, Roll=2.6429, Yaw=-180) {enabled=false, module=AdjustTransform002}
   &self:AdjustTransform(PositionOffset:Vector3f=<0 1 0>) {enabled=false, module=AdjustTransform}
  ]
  particleUpdateScript [
   &ParticleState()
   &ScaleColor(
    Scale Alpha=&FloatFromCurve(FloatCurve=DataInterface() {propertyValues {Curve {keys [
         {arriveTangent=9.0725, interpMode=RCIM_Cubic, leaveTangent=9.0725, tangentMode=RCTM_User, time=0, value=1.05}
         {arriveTangent=-2.5249, interpMode=RCIM_Cubic, leaveTangent=-2.5249, tangentMode=RCTM_User, time=0.3, value=1.975}
         {arriveTangent=-0.0405, interpMode=RCIM_Cubic, leaveTangent=-0.0405, tangentMode=RCTM_User, time=1, value=0}
        ]}}})
    ScaleA=true
    ScaleRGB=false
   )
   &DynamicMaterialParameters(
    Index 0 Param 1=&FloatFromCurve(FloatCurve=DataInterface() {propertyValues {Curve {keys [
         {arriveTangent=-0.0674, interpMode=RCIM_Cubic, leaveTangent=-0.0674, tangentMode=RCTM_User, time=0.4, value=0}
         {arriveTangent=4.1799, interpMode=RCIM_Cubic, leaveTangent=4.1799, tangentMode=RCTM_User, time=0.8, value=1.125}
         {arriveTangent=-0.0405, interpMode=RCIM_Cubic, leaveTangent=-0.0405, tangentMode=RCTM_User, time=1, value=0}
        ]}}})
    Index 0 Param 2=0.3
    Index 0 Param 3=0.2
    Module.Param0WriteEnabled=true
    Module.Param1WriteEnabled=true
    Module.Param2WriteEnabled=true
    Module.Param3WriteEnabled=false
   )
   &FN_Mesh_Attributes(Create Orientation Attribute=true, Create Scale Attribute=true, Orientation:Quat4f=$Engine.Owner.Rotation, Scale=<1.4 1.4 1.4>) {enabled=false}
  ]
  rendererClasses [&NiagaraMeshRendererProperties]
  renderers [{rendererClass=&NiagaraMeshRendererProperties, rendererIndex=0}]
 }
 Swirly_Runes {
  simTarget=GPUComputeSim
  emitterUpdateScript [
   &EmitterState(Life Cycle Mode=::NewEnumerator1, Loop Behavior=::NewEnumerator1, Loop Duration=1, UseLoopDelay=false)
   &SpawnBurst_Instantaneous(Spawn Count=35, Use Spawn Probability=false, UseLoopCountLimit=false)
  ]
  particleSpawnScript [
   &InitializeParticle(
    Color Maximum=<rgba 0 1 0.137 1>
    Color Minimum=<rgba 1 0 0.1271 1>
    Color Mode=::NewEnumerator2
    Lifetime Min=0.5
    Lifetime Mode=::NewEnumerator1
    Mass Max=5
    Mass Min=0.4
    Mass Mode=::NewEnumerator3
    Sprite Rotation Mode=::NewEnumerator2
    Sprite Size Mode=::NewEnumerator1
    Uniform Sprite Size Max=5
    Uniform Sprite Size Min=2
    UsePositionOffset=false
    Write Material Random=true
   )
   &SphereLocation(
    MaskSpawnGroup=false
    Offset=<0 0.5 -20>
    Override Seed=false
    Sphere Radius=5
    Surface Only=false
    UseNonUniformScale=false
    UseOffset=true
    UseSphereOrientation=false
   )
   &AddVelocityFromPoint(
    Module.Use Falloff Distance=false
    Origin Offset=<0 5 -22>
    Velocity Origin:Vector3f=/Niagara/DynamicInputs/Helpers/SimulationPosition()
    Velocity Strength=&RandomRangeFloat(Minimum=15, Maximum=25)
   )
   &AddVelocity(Velocity=&RandomRangeVector(Minimum=<0 -35 0>, Maximum=<0 -60 0>))
   /SetVariables(Particles.RandRotationRate:NiagaraFloat=&RandomRangeFloat(Minimum=-5, Maximum=5)) {isSetParameters=true, module=SetVariables_#2}
   &self:Clamp_Speed(MaxSpeed=44) {module=Clamp_Speed}
   /SetVariables(Particles.Speed:NiagaraFloat=&VectorLength(Vector=$Particles.Velocity)) {isSetParameters=true, module=SetVariables_#3}
  ]
  particleUpdateScript [
   &ParticleState()
   &GravityForce(Gravity=<0 0 -50>) {enabled=false}
   &Drag(Drag=&RandomRangeFloat(Minimum=0, Maximum=0.3), Use Linear Drag=true, Use Rotational Drag=false)
   &ScaleSpriteSize(Scale Factor=&Vector2DFromFloat(Value=&FloatFromCurve(
      FloatCurve=DataInterface() {propertyValues {Curve {keys [
          <at 0 -0.05>
          {arriveTangent=1.3169, interpMode=RCIM_Cubic, leaveTangent=1.3169, time=0.1, value=1}
          {arriveTangent=-1.1075, interpMode=RCIM_Cubic, leaveTangent=-1.1075, time=0.7001, value=0.872}
          <at 1.0011 0.002>
         ]}}}
      Scale Curve=0.3
     )))
   &ScaleColor(
    Scale Alpha=&FloatFromCurve(
     FloatCurve=DataInterface() {propertyValues {Curve {keys [<at 0 -0.05>, {arriveTangent=0.0519, interpMode=RCIM_Cubic, leaveTangent=0.0519, time=0.1, value=1}, <at 1.0011 0.002 RCIM_Cubic>]}}}
     Scale Curve=3
    )
    Scale RGB=&VectorFromFloat(Value=&FloatFromCurve(FloatCurve=DataInterface() {propertyValues {Curve {keys [<at 0 1>,<at 1 0>]}}}, Scale Curve=5))
    ScaleA=true
    ScaleRGB=true
   )
   &CurlNoiseForce(
    Noise Frequency=80
    Noise Strength=&FloatFromCurve(
     FloatCurve=DataInterface() {propertyValues {Curve {keys [
         {arriveTangent=4.1785, interpMode=RCIM_Cubic, leaveTangent=4.1785, tangentMode=RCTM_User, time=0, value=0}
         {arriveTangent=0.9995, interpMode=RCIM_Cubic, leaveTangent=0.9995, tangentMode=RCTM_User, time=0.1, value=0.85}
         {arriveTangent=-0.0405, interpMode=RCIM_Cubic, leaveTangent=-0.0405, tangentMode=RCTM_User, time=1, value=1}
        ]}}}
     Scale Curve=&Multiply_Float(A=12, B=$Particles.Speed)
    )
    Pan Noise=true
    Pan Noise Field=<0 55 0>
   )
   &ScaleColor(
    Scale Alpha=&FloatFromCurve(FloatCurve=DataInterface() {propertyValues {Curve {keys [{interpMode=RCIM_Cubic, tangentMode=RCTM_User, time=0.75, value=1}, <at 1 0.5>]}}})
    Scale RGB=&MakeVector(
     X=&UniformAOrBFloat(A=0)
     Y=1
     Z=&FloatFromCurve(FloatCurve=DataInterface() {propertyValues {Curve {keys [
          {arriveTangent=-32.9921, interpMode=RCIM_Cubic, leaveTangent=-32.9921, tangentMode=RCTM_User, time=0.2, value=10}
          {arriveTangent=-0.0507, interpMode=RCIM_Cubic, leaveTangent=-0.0508, tangentMode=RCTM_User, time=0.3, value=1}
          {arriveTangent=8.5796, interpMode=RCIM_Cubic, leaveTangent=8.5796, tangentMode=RCTM_User, time=1, value=20}
         ]}}})
    )
    ScaleA=true
    ScaleRGB=true
   ) {enabled=false, module=ScaleColor001}
   &SpriteRotationRate(Rotation Rate=&Multiply_Float(A=$Particles.RandRotationRate, B=&VectorLength(Vector=$Particles.Velocity)))
   &SolveForcesAndVelocity(Clamp Velocity=false, Limit Acceleration=false)
   &DynamicMaterialParameters(
    Index 0 Param 1=0.5
    Index 0 Param 2=5
    Module.Param0WriteEnabled=true
    Module.Param1WriteEnabled=true
    Module.Param2WriteEnabled=false
    Module.Param3WriteEnabled=false
   )
  ]
  rendererClasses [&NiagaraSpriteRendererProperties]
  renderers [{rendererClass=&NiagaraSpriteRendererProperties}]
 }
 Swirly_Runes001 {
  bEnabled=false
  simTarget=GPUComputeSim
  emitterUpdateScript [
   &EmitterState(Life Cycle Mode=::NewEnumerator1, Loop Behavior=::NewEnumerator1, Loop Duration=1, UseLoopDelay=false)
   &SpawnBurst_Instantaneous(Spawn Count=35, Use Spawn Probability=false, UseLoopCountLimit=false)
  ]
  particleSpawnScript [
   &InitializeParticle(
    Color Maximum=<rgba 0 1 0.137 1>
    Color Minimum=<rgba 1 0 0.1271 1>
    Color Mode=::NewEnumerator2
    Lifetime Min=0.5
    Lifetime Mode=::NewEnumerator1
    Mass Max=5
    Mass Min=0.4
    Mass Mode=::NewEnumerator3
    Sprite Rotation Mode=::NewEnumerator2
    Sprite Size Mode=::NewEnumerator1
    Uniform Sprite Size Max=5
    Uniform Sprite Size Min=2
    UsePositionOffset=false
    Write Material Random=true
   )
   &SphereLocation(
    MaskSpawnGroup=false
    Offset=<0 0.5 0>
    Override Seed=false
    Sphere Radius=5
    Surface Only=false
    UseNonUniformScale=false
    UseOffset=true
    UseSphereOrientation=false
   )
   &AddVelocityFromPoint(
    Module.Use Falloff Distance=false
    Origin Offset=<0 3 -2>
    Velocity Origin:Vector3f=&SimulationPosition()
    Velocity Strength=&RandomRangeFloat(Minimum=3, Maximum=25)
   )
   &AddVelocity(Velocity=&RandomRangeVector(Minimum=<0 -35 0>, Maximum=<0 -60 0>))
   /SetVariables(Particles.RandRotationRate:NiagaraFloat=&RandomRangeFloat(Minimum=-5, Maximum=5)) {isSetParameters=true, module=SetVariables_#2}
   /SetVariables(Particles.Speed:NiagaraFloat=&VectorLength(Vector=$Particles.Velocity)) {isSetParameters=true, module=SetVariables_#3}
  ]
  particleUpdateScript [
   &ParticleState()
   &GravityForce(Gravity=<0 0 -50>) {enabled=false}
   &Drag(Drag=&RandomRangeFloat(Minimum=0, Maximum=0.3), Use Linear Drag=true, Use Rotational Drag=false)
   &ScaleSpriteSize(Scale Factor=&Vector2DFromFloat(Value=&FloatFromCurve(
      FloatCurve=DataInterface() {propertyValues {Curve {keys [
          <at 0 -0.05>
          {arriveTangent=1.3169, interpMode=RCIM_Cubic, leaveTangent=1.3169, time=0.1, value=1}
          {arriveTangent=-1.1075, interpMode=RCIM_Cubic, leaveTangent=-1.1075, time=0.7001, value=0.872}
          <at 1.0011 0.002>
         ]}}}
      Scale Curve=0.3
     )))
   &ScaleColor(
    Scale Alpha=&FloatFromCurve(
     FloatCurve=DataInterface() {propertyValues {Curve {keys [<at 0 -0.05>, {arriveTangent=0.0519, interpMode=RCIM_Cubic, leaveTangent=0.0519, time=0.1, value=1}, <at 1.0011 0.002 RCIM_Cubic>]}}}
     Scale Curve=3
    )
    Scale RGB=&VectorFromFloat(Value=&FloatFromCurve(FloatCurve=DataInterface() {propertyValues {Curve {keys [<at 0 1>,<at 1 0>]}}}, Scale Curve=5))
    ScaleA=true
    ScaleRGB=true
   )
   &CurlNoiseForce(
    Noise Frequency=80
    Noise Strength=&FloatFromCurve(
     FloatCurve=DataInterface() {propertyValues {Curve {keys [
         {arriveTangent=4.1785, interpMode=RCIM_Cubic, leaveTangent=4.1785, tangentMode=RCTM_User, time=0, value=0}
         {arriveTangent=0.9995, interpMode=RCIM_Cubic, leaveTangent=0.9995, tangentMode=RCTM_User, time=0.1, value=0.85}
         {arriveTangent=-0.0405, interpMode=RCIM_Cubic, leaveTangent=-0.0405, tangentMode=RCTM_User, time=1, value=1}
        ]}}}
     Scale Curve=&Multiply_Float(A=12, B=$Particles.Speed)
    )
    Pan Noise=true
    Pan Noise Field=<0 55 0>
   )
   &ScaleColor(
    Scale Alpha=&FloatFromCurve(FloatCurve=DataInterface() {propertyValues {Curve {keys [{interpMode=RCIM_Cubic, tangentMode=RCTM_User, time=0.75, value=1}, <at 1 0.5>]}}})
    Scale RGB=&MakeVector(
     X=&UniformAOrBFloat(A=0)
     Y=1
     Z=&FloatFromCurve(FloatCurve=DataInterface() {propertyValues {Curve {keys [
          {arriveTangent=-32.9921, interpMode=RCIM_Cubic, leaveTangent=-32.9921, tangentMode=RCTM_User, time=0.2, value=10}
          {arriveTangent=-0.0507, interpMode=RCIM_Cubic, leaveTangent=-0.0508, tangentMode=RCTM_User, time=0.3, value=1}
          {arriveTangent=8.5796, interpMode=RCIM_Cubic, leaveTangent=8.5796, tangentMode=RCTM_User, time=1, value=20}
         ]}}})
    )
    ScaleA=true
    ScaleRGB=true
   ) {enabled=false, module=ScaleColor001}
   &SpriteRotationRate(Rotation Rate=&Multiply_Float(A=$Particles.RandRotationRate, B=&VectorLength(Vector=$Particles.Velocity)))
   &SolveForcesAndVelocity(Clamp Velocity=false, Limit Acceleration=false)
   &DynamicMaterialParameters(
    Index 0 Param 1=0.5
    Index 0 Param 2=5
    Module.Param0WriteEnabled=true
    Module.Param1WriteEnabled=true
    Module.Param2WriteEnabled=false
    Module.Param3WriteEnabled=false
   )
  ]
  rendererClasses [&NiagaraSpriteRendererProperties]
  renderers [{rendererClass=&NiagaraSpriteRendererProperties}]
 }
 Vertical_Beam {
  bEnabled=false
  emitterUpdateScript [
   &EmitterState()
   &SpawnRate(Use Spawn Probability=false) {enabled=false}
   &SpawnBurst_Instantaneous(Spawn Count=1, Use Spawn Probability=false)
  ]
  particleSpawnScript [&InitializeParticle(Color Mode=::NewEnumerator0, Lifetime=0.75, UsePositionOffset=false, Write Material Random=true)]
  particleUpdateScript [
   &ParticleState()
   &Color(Color=&ColorFromCurve(ColorCurve=DataInterface() {propertyValues {
       AlphaCurve {keys [<at 0.657 1 RCIM_Cubic>,<at 1 0 RCIM_Cubic>]}
       BlueCurve {keys [<at 0.0549 0>, {arriveTangent=1.396, interpMode=RCIM_Cubic, leaveTangent=1.396, time=0.2989, value=0.0194}, <at 0.7712 1 RCIM_Cubic>]}
       GreenCurve {keys [
         <at 0.0549 1>
         {arriveTangent=-1.2524, interpMode=RCIM_Cubic, leaveTangent=-1.2524, time=0.2989, value=0.0692}
         <at 0.7712 0.1028 RCIM_Cubic>
        ]}
       RedCurve {keys [<at 0.0549 3>, {arriveTangent=-3.765, interpMode=RCIM_Cubic, leaveTangent=-3.765, time=0.2989, value=1}, <at 0.7712 0.303 RCIM_Cubic>]}
      }}))
   /SetVariables(Particles.SpriteSize=&Multiply_Vector2D_ByFloat(
     Float=&FloatFromCurve(FloatCurve=DataInterface() {propertyValues {Curve {keys [<at 0 0.6>,<at 0.1 0.95>,<at 0.7 0.75>,<at 1 0>]}}})
     Vector2D=<100 400>
    )) {isSetParameters=true, module=SetVariables_#1}
   &FN_Radial_Material(
    Radial Mask Power=1
    Radial Mask Tile U=0.2
    Radial Mask Tile V=0.5
    Sprite Extrusion Speed:NiagaraFloat=&FloatFromCurve(
     FloatCurve=DataInterface() {propertyValues {Curve {keys [
         {arriveTangent=-0.2256, interpMode=RCIM_Cubic, time=0, value=1.75}
         {arriveTangent=2.75, interpMode=RCIM_Cubic, leaveTangent=2.75, time=0.3, value=2.375}
         {arriveTangent=2.5, interpMode=RCIM_Cubic, leaveTangent=2.5, time=0.5, value=3.125}
         {interpMode=RCIM_Cubic, leaveTangent=-0.5666, time=0.7, value=3.375}
        ]}}}
     Scale Curve=-0.3
    )
    Sprite Spin Speed:NiagaraFloat=&FloatFromCurve(FloatCurve=DataInterface() {propertyValues {Curve {keys [<at 0 0.2>,<at 0.4 1.125>,<at 1 2.25>]}}}, Scale Curve=0.1)
   ) {enabled=false}
  ]
  rendererClasses [&NiagaraSpriteRendererProperties]
  renderers [
   {rendererClass=&NiagaraSpriteRendererProperties}
   {rendererClass=&NiagaraSpriteRendererProperties, rendererIndex=1}
  ]
 }
}
summary {
 systemName=NS_SidewaysChest_LockOpen
 emitters [
  {emitterName=LightRenderer, rendererClasses [&NiagaraLightRendererProperties]}
  {emitterName=Swirly_Runes, simTarget=GPUComputeSim, rendererClasses [&NiagaraSpriteRendererProperties]}
  {emitterName=Mesh_rays001, rendererClasses [&NiagaraSpriteRendererProperties, &NiagaraMeshRendererProperties]}
  {bEnabled=false, emitterName=Halo_and_Glow, rendererClasses [&NiagaraSpriteRendererProperties]}
  {bEnabled=false, emitterName=Vertical_Beam, rendererClasses [&NiagaraSpriteRendererProperties]}
  {emitterName=Mesh_rays_offset, rendererClasses [&NiagaraMeshRendererProperties]}
  {bEnabled=false, emitterName=Swirly_Runes001, simTarget=GPUComputeSim, rendererClasses [&NiagaraSpriteRendererProperties]}
  {
   bEnabled=false
   emitterName=Mesh_rays
   rendererClasses [&NiagaraSpriteRendererProperties, &NiagaraMeshRendererProperties]
  }
  {bEnabled=false, emitterName=Mesh_rays_offset001, rendererClasses [&NiagaraMeshRendererProperties]}
 ]
 userVariables []
}
