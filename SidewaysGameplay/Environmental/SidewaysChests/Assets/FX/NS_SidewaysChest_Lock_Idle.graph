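# Niagara idle effect for the lock on a Sideways chest, showing a pinkish-red light and ray meshes.
# LightRenderer and Mesh_rays_offset each burst-spawn with Particle State disabled, so particles persist without aging out.
# The rays get a position offset and roll from a scratch module, plus alpha and dynamic material parameters for their look.
# The GPU Swirly_Runes sprite emitter and the original Mesh_rays emitter are disabled.
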NiagaraSystem /SidewaysGameplay/Environmental/SidewaysChests/Assets/FX/NS_SidewaysChest_Lock_Idle
dependencies {
 usedDataInterfaces [&NiagaraDataInterfaceCurve, &NiagaraDataInterfaceColorCurve]
 usedDynamicInputs []
 usedModules [&SystemState]
}
emitters {
 LightRenderer {
  emitterUpdateScript [
   &EmitterState(Life Cycle Mode=::NewEnumerator1, Loop Behavior=::NewEnumerator1, Loop Duration=2, UseLoopDelay=false)
   &SpawnBurst_Instantaneous(Use Spawn Probability=false, UseLoopCountLimit=false)
  ]
  particleSpawnScript [
   &InitializeParticle(
    Color=<rgba 1 0.0439 0.1636 4>
    Lifetime=0.5
    Position Offset=<0 0 -15>
    Sprite Size Mode=::NewEnumerator1
    Uniform Sprite Size Max=100
    Uniform Sprite Size Min=55
    UsePositionOffset=true
    Write Material Random=true
   )
   &SphereLocation(
    MaskSpawnGroup=false
    Override Seed=false
    Sphere Radius=2
    Surface Only=false
    UseNonUniformScale=false
    UseOffset=false
    UseSphereOrientation=false
   ) {enabled=false}
   /SetVariables(Particles.LightRadius=25, Particles.LightScale:Vector3f=<1 1 1>) {isSetParameters=true}
  ]
  particleUpdateScript [
   &ParticleState() {enabled=false}
   &ScaleColor(
    Scale Alpha=&FloatFromCurve(FloatCurve=DataInterface() {propertyValues {Curve {keys [<at 0.2 2>,<at 0.7 0.5>,<at 1 0>]}}})
    ScaleA=true
    ScaleRGB=true
   ) {enabled=false}
  ]
  rendererClasses [&NiagaraLightRendererProperties]
  renderers [{rendererClass=&NiagaraLightRendererProperties, rendererIndex=0}]
 }
 Mesh_rays {
  bEnabled=false
  emitterUpdateScript [
   &EmitterState(Life Cycle Mode=::NewEnumerator1, Loop Behavior=::NewEnumerator1, Loop Duration=2, UseLoopDelay=false)
   &SpawnBurst_Instantaneous(Use Spawn Probability=false, UseLoopCountLimit=false)
  ]
  particleSpawnScript [
   &InitializeParticle(Mesh Scale Mode=::NewEnumerator3, UsePositionOffset=false, Write Material Random=true)
   &FN_Mesh_Attributes(Create Orientation Attribute=true, Create Position Attribute=true) {module=FN_Mesh_Attributes001}
  ]
  particleUpdateScript [
   &ParticleState() {enabled=false}
   &ScaleColor(
    Scale Alpha=&FloatFromCurve(FloatCurve=DataInterface() {propertyValues {Curve {keys [<at 0.2 1>,<at 0.4 0.35>,<at 0.6 0.05>,<at 1 0>]}}})
    ScaleA=true
    ScaleRGB=false
   )
   &DynamicMaterialParameters(
    Index 0 Param 1=&FloatFromCurve(FloatCurve=DataInterface() {propertyValues {Curve {keys [
         {arriveTangent=-0.0674, interpMode=RCIM_Cubic, leaveTangent=-0.0674, tangentMode=RCTM_User, time=0.4, value=0}
         {arriveTangent=4.1799, interpMode=RCIM_Cubic, leaveTangent=4.1799, tangentMode=RCTM_User, time=0.8, value=1.125}
         {arriveTangent=-0.0405, interpMode=RCIM_Cubic, leaveTangent=-0.0405, tangentMode=RCTM_User, time=1, value=0}
        ]}}})
    Index 0 Param 2=0.3
    Index 0 Param 3=0.2
    Module.Param0WriteEnabled=true
    Module.Param1WriteEnabled=true
    Module.Param2WriteEnabled=true
    Module.Param3WriteEnabled=false
   )
   &FN_Mesh_Attributes(Create Orientation Attribute=true, Create Scale Attribute=true, Orientation:Quat4f=$Engine.Owner.Rotation, Scale=<1.4 1.4 1.4>) {enabled=false}
  ]
  rendererClasses [&NiagaraMeshRendererProperties]
  renderers [{rendererClass=&NiagaraMeshRendererProperties, rendererIndex=0}]
 }
 Mesh_rays_offset {
  emitterUpdateScript [
   &EmitterState(Life Cycle Mode=::NewEnumerator1, Loop Behavior=::NewEnumerator1, Loop Duration=2, UseLoopDelay=false)
   &SpawnBurst_Instantaneous(Use Spawn Probability=false, UseLoopCountLimit=false)
  ]
  particleSpawnScript [
   &InitializeParticle(Mesh Scale Mode=::NewEnumerator3, UsePositionOffset=false, Write Material Random=true)
   &FN_Mesh_Attributes() {enabled=false, module=FN_Mesh_Attributes001}
   &self:ScratchModule_03(PositionOffset:Vector3f=<0 1 0>, Roll=20) {module=ScratchModule_03}
  ]
  particleUpdateScript [
   &ParticleState() {enabled=false}
   &ScaleColor(
    Scale Alpha=&FloatFromCurve(FloatCurve=DataInterface() {propertyValues {Curve {keys [<at 0.2 1>,<at 0.4 0.35>,<at 0.6 0.05>,<at 1 0>]}}})
    ScaleA=true
    ScaleRGB=false
   )
   &DynamicMaterialParameters(
    Index 0 Param 1=&FloatFromCurve(FloatCurve=DataInterface() {propertyValues {Curve {keys [
         {arriveTangent=-0.0674, interpMode=RCIM_Cubic, leaveTangent=-0.0674, tangentMode=RCTM_User, time=0.4, value=0}
         {arriveTangent=4.1799, interpMode=RCIM_Cubic, leaveTangent=4.1799, tangentMode=RCTM_User, time=0.8, value=1.125}
         {arriveTangent=-0.0405, interpMode=RCIM_Cubic, leaveTangent=-0.0405, tangentMode=RCTM_User, time=1, value=0}
        ]}}})
    Index 0 Param 2=0.3
    Index 0 Param 3=0.2
    Module.Param0WriteEnabled=true
    Module.Param1WriteEnabled=true
    Module.Param2WriteEnabled=true
    Module.Param3WriteEnabled=false
   )
   &FN_Mesh_Attributes(Create Orientation Attribute=true, Create Scale Attribute=true, Orientation:Quat4f=$Engine.Owner.Rotation, Scale=<1.4 1.4 1.4>) {enabled=false}
  ]
  rendererClasses [&NiagaraMeshRendererProperties]
  renderers [{rendererClass=&NiagaraMeshRendererProperties, rendererIndex=0}]
 }
 Swirly_Runes {
  bEnabled=false
  simTarget=GPUComputeSim
  emitterUpdateScript [
   &EmitterState(Life Cycle Mode=::NewEnumerator1, Loop Duration=1, UseLoopDelay=false)
   &SpawnBurst_Instantaneous(Spawn Count=35, Use Spawn Probability=false, UseLoopCountLimit=false) {enabled=false}
   &SpawnRate(SpawnRate=5, Use Spawn Probability=false)
  ]
  particleSpawnScript [
   &InitializeParticle(
    Color Maximum=<rgba 0 1 0.137 1>
    Color Minimum=<rgba 1 0 0.1271 1>
    Color Mode=::NewEnumerator2
    Lifetime Max=1.5
    Lifetime Min=0.3
    Lifetime Mode=::NewEnumerator1
    Mass Max=5
    Mass Min=0.4
    Mass Mode=::NewEnumerator3
    Sprite Rotation Mode=::NewEnumerator2
    Sprite Size Mode=::NewEnumerator1
    Uniform Sprite Size Max=8
    Uniform Sprite Size Min=3
    UsePositionOffset=false
    Write Material Random=true
   )
   &SphereLocation(
    MaskSpawnGroup=false
    Offset=<0 0 -5>
    Override Seed=false
    Sphere Radius=15
    Surface Only=false
    UseNonUniformScale=false
    UseOffset=true
    UseSphereOrientation=false
   )
   &AddVelocityFromPoint(
    Module.Use Falloff Distance=false
    Origin Offset=<0 0 -20>
    Velocity Origin:Vector3f=/Niagara/DynamicInputs/Helpers/SimulationPosition()
    Velocity Strength=&RandomRangeFloat(Minimum=125, Maximum=555)
   ) {enabled=false}
   &AddVelocity(Velocity=<0 -45 33>)
   /SetVariables(Particles.RandRotationRate:NiagaraFloat=&RandomRangeFloat(Minimum=-5, Maximum=5)) {isSetParameters=true}
  ]
  particleUpdateScript [
   &ParticleState()
   &GravityForce(Gravity=<0 0 -450>) {enabled=false}
   &Drag(Drag=0.75, Use Linear Drag=true, Use Rotational Drag=false) {enabled=false}
   &ScaleSpriteSize(Scale Factor=&Vector2DFromFloat(Value=&FloatFromCurve(
      FloatCurve=DataInterface() {propertyValues {Curve {keys [
          <at 0 -0.2>
          {arriveTangent=1.5312, interpMode=RCIM_Cubic, leaveTangent=1.5312, time=0.1, value=1}
          {arriveTangent=-1.1075, interpMode=RCIM_Cubic, leaveTangent=-1.1075, time=0.7001, value=0.872}
          <at 1.0011 0.002>
         ]}}}
      Scale Curve=1.3
     )))
   &ScaleColor(
    Scale Alpha=&FloatFromCurve(FloatCurve=DataInterface() {propertyValues {Curve {keys [<at 0 0>, {arriveTangent=0.002, interpMode=RCIM_Cubic, leaveTangent=0.002, time=0.1, value=0.95}, <at 1.0011 0.002 RCIM_Cubic>]}}})
    Scale RGB=&VectorFromFloat(Value=&FloatFromCurve(FloatCurve=DataInterface() {propertyValues {Curve {keys [<at 0 1>,<at 1 0>]}}}, Scale Curve=24))
    ScaleA=true
    ScaleRGB=true
   )
   &CurlNoiseForce(Noise Strength=444, Pan Noise=false)
   &ScaleColor(
    Scale Alpha=&FloatFromCurve(FloatCurve=DataInterface() {propertyValues {Curve {keys [{interpMode=RCIM_Cubic, tangentMode=RCTM_User, time=0.75, value=1}, <at 1 0.5>]}}})
    Scale RGB=&MakeVector(
     X=&UniformAOrBFloat(A=0)
     Y=1
     Z=&FloatFromCurve(FloatCurve=DataInterface() {propertyValues {Curve {keys [
          {arriveTangent=-32.9921, interpMode=RCIM_Cubic, leaveTangent=-32.9921, tangentMode=RCTM_User, time=0.2, value=10}
          {arriveTangent=-0.0507, interpMode=RCIM_Cubic, leaveTangent=-0.0508, tangentMode=RCTM_User, time=0.3, value=1}
          {arriveTangent=8.5796, interpMode=RCIM_Cubic, leaveTangent=8.5796, tangentMode=RCTM_User, time=1, value=20}
         ]}}})
    )
    ScaleA=true
    ScaleRGB=true
   ) {enabled=false, module=ScaleColor001}
   &SpriteRotationRate(Rotation Rate=&Multiply_Float(A=$Particles.RandRotationRate, B=&VectorLength(Vector=$Particles.Velocity)))
   &SolveForcesAndVelocity(Clamp Velocity=false, Limit Acceleration=false)
   &DynamicMaterialParameters(
    Index 0 Param 1=0.5
    Index 0 Param 2=5
    Module.Param0WriteEnabled=true
    Module.Param1WriteEnabled=true
    Module.Param2WriteEnabled=false
    Module.Param3WriteEnabled=false
   )
  ]
  rendererClasses [&NiagaraSpriteRendererProperties]
  renderers [{rendererClass=&NiagaraSpriteRendererProperties}]
 }
}
summary {
 systemName=NS_SidewaysChest_Lock_Idle
 emitters [
  {emitterName=LightRenderer, rendererClasses [&NiagaraLightRendererProperties]}
  {bEnabled=false, emitterName=Swirly_Runes, simTarget=GPUComputeSim, rendererClasses [&NiagaraSpriteRendererProperties]}
  {bEnabled=false, emitterName=Mesh_rays, rendererClasses [&NiagaraMeshRendererProperties]}
  {emitterName=Mesh_rays_offset, rendererClasses [&NiagaraMeshRendererProperties]}
 ]
 userVariables []
}
